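# Sound cue for the first-person fire-stop sound of the Figment compact SMG (PDW) in Athena.
# A single wave player plays the PDW fire-stop wave through a Modulator node, with the cue's volume reduced.
# It routes through the first-person ranged weapon fire sound class with first-person projectile attenuation and gunfire concurrency, and has debug enabled.
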SoundCue /Figment_CompactSMG_Athena/Audio/PDW_Fire_Stop_1P_Figment_Cue
root=Modulator_0
nodes [
 {id=Modulator_0, children [WavePlayer_1]}
 {id=WavePlayer_1, properties {soundWaveAssetPtr=&/PDW_Fire_Stop_01}}
]
properties {
 attenuationSettings=&Fort_Weapon_Projectile_FirstPerson_Athena_Att
 bDebug=true
 maxDistance=2097152
 soundClassObject=&Fort_WeaponRanged_Fire_1P_PSM
 volumeMultiplier=0.5
 concurrencySet [&Weapon_Gunfire_FirstPerson_Conc]
}
