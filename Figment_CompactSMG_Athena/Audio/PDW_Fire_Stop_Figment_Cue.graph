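# Sound cue for the Figment compact SMG (PDW) fire-stop sound, played as third-person gunfire.
# A mixer layers close, medium and distant waves at separate levels, each with its own attenuation settings.
# Close and medium layers use the third-person fire sound class, while the distant layer uses a distant class with raised, randomized pitch.
# Local and global gunfire concurrency limit playback, and the debug flag is left on.
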SoundCue /Figment_CompactSMG_Athena/Audio/PDW_Fire_Stop_Figment_Cue
root=Mixer_17
nodes [
 {
  id=Attenuation_0
  children [Modulator_2]
  properties {attenuationSettings=&Fort_Weapon_Projectile_Medium_Mono_Athena_Att}
 }
 {
  id=Attenuation_34
  children [Modulator_12]
  properties {attenuationSettings=&Fort_Weapon_Projectile_Distant_Mono_Athena_Att}
 }
 {
  id=Attenuation_49
  children [WavePlayer_1]
  properties {attenuationSettings=&Fort_Weapon_Projectile_Close_Athena_Att}
 }
 {id=Mixer_17, children [SoundClass_32, SoundClass_0, SoundClass_35], properties {inputVolume [1,0.15,0.75]}}
 {id=Modulator_0, children [Attenuation_49]}
 {id=Modulator_12, children [WavePlayer_2], properties {pitchMax=1.2, pitchMin=1.1, volumeMax=2, volumeMin=2}}
 {id=Modulator_2, children [WavePlayer_3], properties {volumeMax=2, volumeMin=2}}
 {
  id=SoundClass_0
  children [Attenuation_0]
  properties {soundClassOverride=&Fort_WeaponRanged_Fire_3P_PSM}
 }
 {
  id=SoundClass_32
  children [Modulator_0]
  properties {soundClassOverride=&Fort_WeaponRanged_Fire_3P_PSM}
 }
 {
  id=SoundClass_35
  children [Attenuation_34]
  properties {soundClassOverride=&Fort_WeaponRanged_Fire_Distant}
 }
 {id=WavePlayer_1, properties {soundWaveAssetPtr=&/PDW_Fire_Stop_01}}
 {id=WavePlayer_2, properties {soundWaveAssetPtr=&/PDW_Fire_Stop_Distant_01}}
 {id=WavePlayer_3, properties {soundWaveAssetPtr=&/PDW_Fire_Stop_Medium_01}}
]
properties {
 bDebug=true
 maxDistance=27500
 soundClassObject=&Fort_WeaponRanged_Fire_3P_PSM
 volumeMultiplier=0.5
 concurrencySet [&Weapon_Gunfire_Local_Conc, &Weapon_Gunfire_Global_Conc]
}
