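# Sound cue for picking up the Figment compact SMG (PDW) weapon.
# A single wave player plays the PDW pickup wave at slightly lowered pitch and reduced volume, through the shared weapon pickup attenuation and loot collect sound class.
# A concurrency override allows one instance at a time and blocks new plays while one is active.
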SoundCue /Figment_CompactSMG_Athena/Audio/PDW_Pickup_Figment_Cue
root=WavePlayer_1
nodes [{id=WavePlayer_1, properties {soundWaveAssetPtr=&/PDW_Pickup_01}}]
properties {
 attenuationSettings=&Fort_Weapon_Pickup_Att
 bOverrideConcurrency=true
 maxDistance=3500
 pitchMultiplier=0.95
 soundClassObject=&Fort_Loot_Collect
 volumeMultiplier=0.618
 concurrencyOverrides {maxCount=1, resolutionRule=PreventNew}
}
