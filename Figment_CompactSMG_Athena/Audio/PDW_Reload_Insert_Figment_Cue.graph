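# Sound cue for the magazine insert foley during the Figment compact SMG reload, built from PDW reload waves.
# A mixer layers the insert wave with a quieter slam wave, which is pitched down and delayed by a fixed interval to land after it.
# It uses weapon reload attenuation and the weapon foley sound class, with a concurrency override that prevents new instances once the limit is reached.
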SoundCue /Figment_CompactSMG_Athena/Audio/PDW_Reload_Insert_Figment_Cue
root=Mixer_0
nodes [
 {id=Delay_0, children [Modulator_1], properties {delayMax=0.7, delayMin=0.7}}
 {id=Mixer_0, children [WavePlayer_6,Delay_0], properties {inputVolume [1,0.744]}}
 {id=Modulator_1, children [WavePlayer_5], properties {pitchMax=0.9, pitchMin=0.9}}
 {id=WavePlayer_5, properties {soundWaveAssetPtr=&/PDW_Reload_InsertSlam_01}}
 {id=WavePlayer_6, properties {soundWaveAssetPtr=&/PDW_Reload_Insert_01}}
]
properties {
 attenuationSettings=&Fort_Weapon_Reload_Att
 bOverrideConcurrency=true
 maxDistance=2250
 soundClassObject=&Fort_WeaponFoley
 volumeMultiplier=0.501
 attenuationOverrides {falloffDistance=4100, attenuationShapeExtents {x=900}}
 concurrencyOverrides {resolutionRule=PreventNew}
}
