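# Sound cue for the equip foley of the Compact SMG (PDW), reusing a reload-end wave rather than a dedicated equip recording.
# A short fixed delay feeds a modulator at default settings, which slightly varies pitch and volume on the single wave player.
# It plays at roughly half volume in the weapon foley sound class with shared reload attenuation.
# Its own concurrency override blocks new instances once the limit is reached.
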SoundCue /Figment_CompactSMG_Athena/Audio/SC_PDW_Equip
root=Delay_0
nodes [
 {id=Delay_0, children [Modulator_1], properties {delayMax=0.1, delayMin=0.1}}
 {id=Modulator_1, children [WavePlayer_1]}
 {id=WavePlayer_1, properties {soundWaveAssetPtr=&/PDW_Reload_End_01}}
]
properties {
 attenuationSettings=&Fort_Weapon_Reload_Att
 bOverrideConcurrency=true
 maxDistance=2250
 soundClassObject=&Fort_WeaponFoley
 volumeMultiplier=0.501
 attenuationOverrides {falloffDistance=4100, attenuationShapeExtents {x=900}}
 concurrencyOverrides {resolutionRule=PreventNew}
}
